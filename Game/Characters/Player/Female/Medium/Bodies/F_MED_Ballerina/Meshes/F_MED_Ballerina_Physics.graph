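# Physics asset for the female medium Ballerina player body, defining the bodies and joints for its dynamic skirt.
# Eight chains ring the skirt, each running from a small base sphere through parent, 01 and 02 capsules linked by angular-limited constraints.
# Swing is tight at the parent and looser below, with twist tightly limited throughout.
# Pelvis and thigh capsules have no constraints, apparently serving as colliders for the skirt.
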PhysicsAsset /Game/Characters/Player/Female/Medium/Bodies/F_MED_Ballerina/Meshes/F_MED_Ballerina_Physics
bodies {
 dyn_skirt_bk_mid_01 [{
   shapeType=Capsule
   length=10.1183
   radius=2.8975
   shapeName=dyn_skirt_bk_mid_01_capsule
   center=<7.3723 0 0>
   rotation=<rot 79.4144 -180 -180>
  }]
 dyn_skirt_bk_mid_02 [{shapeType=Capsule, length=12.0875, radius=4.9609, shapeName=dyn_skirt_bk_mid_02_capsule, center=<7.3723 0 0>, rotation=<rot 90 0 0>}]
 dyn_skirt_bk_mid_base [{shapeType=Sphere, radius=0.5, shapeName=dyn_skirt_bk_mid_base_sphere}]
 dyn_skirt_bk_mid_parent [{shapeType=Capsule, length=7.75, radius=2.75, shapeName=dyn_skirt_bk_mid_parent_capsule, center=<5.9325 0 0>, rotation=<rot 90 0 0>}]
 dyn_skirt_frnt_mid_01 [{shapeType=Capsule, length=8.8691, radius=1.9584, shapeName=dyn_skirt_frnt_mid_01_capsule, center=<7.3723 0 0>, rotation=<rot 90 0 0>}]
 dyn_skirt_frnt_mid_02 [{shapeType=Capsule, length=12.2236, radius=1.5856, shapeName=dyn_skirt_frnt_mid_02_capsule, center=<7.3723 0 0>, rotation=<rot 90 0 0>}]
 dyn_skirt_frnt_mid_base [{shapeType=Sphere, radius=0.5, shapeName=dyn_skirt_frnt_mid_base_sphere}]
 dyn_skirt_frnt_mid_parent [{shapeType=Capsule, length=3.5, radius=2.75, shapeName=dyn_skirt_frnt_mid_parent_capsule, center=<3.9958 0 0>, rotation=<rot 90 0 0>}]
 dyn_skirt_side_bk_l_01 [{shapeType=Capsule, length=5.5408, radius=4.6306, shapeName=dyn_skirt_side_bk_l_01_capsule, center=<7.3723 0 0>, rotation=<rot 90 0 0>}]
 dyn_skirt_side_bk_l_02 [{shapeType=Capsule, length=7.4091, radius=4.591, shapeName=dyn_skirt_side_bk_l_02_capsule, center=<7.3723 0 0>, rotation=<rot 90 0 0>}]
 dyn_skirt_side_bk_l_base [{shapeType=Sphere, radius=0.5, shapeName=dyn_skirt_side_bk_l_base_sphere}]
 dyn_skirt_side_bk_l_parent [{shapeType=Capsule, length=7.75, radius=2.75, shapeName=dyn_skirt_side_bk_l_parent_capsule, center=<5.9325 0 0>, rotation=<rot 90 0 0>}]
 dyn_skirt_side_bk_r_01 [{shapeType=Capsule, length=5.77, radius=4.902, shapeName=dyn_skirt_side_bk_r_01_capsule, center=<7.3723 0 0>, rotation=<rot 90 0 0>}]
 dyn_skirt_side_bk_r_02 [{shapeType=Capsule, length=7.364, radius=5.7137, shapeName=dyn_skirt_side_bk_r_02_capsule, center=<7.3723 0 0>, rotation=<rot 90 0 0>}]
 dyn_skirt_side_bk_r_base [{shapeType=Sphere, radius=0.5, shapeName=dyn_skirt_side_bk_r_base_sphere}]
 dyn_skirt_side_bk_r_parent [{shapeType=Capsule, length=7.75, radius=2.75, shapeName=dyn_skirt_side_bk_r_parent_capsule, center=<5.9325 0 0>, rotation=<rot 90 0 0>}]
 dyn_skirt_side_frnt_l_01 [{
   shapeType=Capsule
   length=8.1162
   radius=2.2543
   shapeName=dyn_skirt_side_frnt_l_01_capsule
   center=<5.2234 0 -1.4778>
   rotation=<rot 74.2027 0 0>
  }]
 dyn_skirt_side_frnt_l_02 [{shapeType=Capsule, length=13.0465, radius=1.7599, shapeName=dyn_skirt_side_frnt_l_02_capsule, center=<7.3723 0 0>, rotation=<rot 90 0 0>}]
 dyn_skirt_side_frnt_l_base [{shapeType=Sphere, radius=0.5, shapeName=dyn_skirt_side_frnt_l_base_sphere}]
 dyn_skirt_side_frnt_l_parent [{shapeType=Capsule, length=3.5, radius=2.75, shapeName=dyn_skirt_side_frnt_l_parent_capsule, center=<3.9958 0 0>, rotation=<rot 90 0 0>}]
 dyn_skirt_side_frnt_r_01 [{shapeType=Capsule, length=8.8566, radius=1.8881, shapeName=dyn_skirt_side_frnt_r_01_capsule, center=<7.3723 0 0>, rotation=<rot 90 0 0>}]
 dyn_skirt_side_frnt_r_02 [{shapeType=Capsule, length=12.8659, radius=1.8918, shapeName=dyn_skirt_side_frnt_r_02_capsule, center=<7.3723 0 0>, rotation=<rot 90 0 0>}]
 dyn_skirt_side_frnt_r_base [{shapeType=Sphere, radius=0.5, shapeName=dyn_skirt_side_frnt_r_base_sphere}]
 dyn_skirt_side_frnt_r_parent [{shapeType=Capsule, length=3.5, radius=2.75, shapeName=dyn_skirt_side_frnt_r_parent_capsule, center=<3.9958 0 0>, rotation=<rot 90 0 0>}]
 dyn_skirt_side_l_01 [{shapeType=Capsule, length=6.9654, radius=4.1446, shapeName=dyn_skirt_side_l_01_capsule, center=<7.3723 0 0>, rotation=<rot 90 0 0>}]
 dyn_skirt_side_l_02 [{shapeType=Capsule, length=8.9606, radius=5.1276, shapeName=dyn_skirt_side_l_02_capsule, center=<7.3723 0 0>, rotation=<rot 90 0 0>}]
 dyn_skirt_side_l_base [{shapeType=Sphere, radius=0.5, shapeName=dyn_skirt_side_l_base_sphere}]
 dyn_skirt_side_l_parent [{shapeType=Capsule, length=3.5, radius=2.75, shapeName=dyn_skirt_side_l_parent_capsule, center=<3.9958 0 0>, rotation=<rot 90 0 0>}]
 dyn_skirt_side_r_01 [{shapeType=Capsule, length=4.1968, radius=5.0145, shapeName=dyn_skirt_side_r_01_capsule, center=<4.8337 0 0>, rotation=<rot 90 0 0>}]
 dyn_skirt_side_r_02 [{shapeType=Capsule, length=9.8342, radius=5.0245, shapeName=dyn_skirt_side_frnt_r_02_capsule, center=<8.6566 0 0>, rotation=<rot 90 0 0>}]
 dyn_skirt_side_r_base [{shapeType=Sphere, radius=0.5, shapeName=dyn_skirt_side_r_base_sphere}]
 dyn_skirt_side_r_parent [{shapeType=Capsule, length=3.5, radius=2.75, shapeName=dyn_skirt_side_r_parent_capsule, center=<3.9958 0 0>, rotation=<rot 90 0 0>}]
 pelvis [{shapeType=Capsule, length=11.788, radius=10.3793, shapeName=pelvis_capsule, center=<-1.9372 -2.5495 0>}]
 thigh_l [{
   shapeType=Capsule
   length=15.6298
   radius=7.3544
   shapeName=thigh_l_capsule
   center=<-17.469 0.0004 0.5605>
   rotation=<rot 84.1817 179.9865 179.9944>
  }]
 thigh_r [{
   shapeType=Capsule
   length=15.6298
   radius=7.3544
   shapeName=thigh_l_capsule
   center=<17.469 -0.0004 -0.5605>
   rotation=<rot 84.1817 179.9865 -0.0056>
  }]
}
constraints [
 {bone1Name=dyn_skirt_frnt_mid_01, bone2Name=dyn_skirt_frnt_mid_parent, swing1LimitDegrees=15, swing2LimitDegrees=10, twistLimitDegrees=2}
 {bone1Name=dyn_skirt_frnt_mid_02, bone2Name=dyn_skirt_frnt_mid_01, swing1LimitDegrees=25, swing2LimitDegrees=20, twistLimitDegrees=2}
 {bone1Name=dyn_skirt_bk_mid_02, bone2Name=dyn_skirt_bk_mid_01, swing1LimitDegrees=25, swing2LimitDegrees=15, twistLimitDegrees=2}
 {bone1Name=dyn_skirt_bk_mid_01, bone2Name=dyn_skirt_bk_mid_parent, swing1LimitDegrees=15, swing2LimitDegrees=25, twistLimitDegrees=2}
 {bone1Name=dyn_skirt_side_bk_l_02, bone2Name=dyn_skirt_side_bk_l_01, swing1LimitDegrees=25, swing2LimitDegrees=15, twistLimitDegrees=2}
 {bone1Name=dyn_skirt_side_bk_l_01, bone2Name=dyn_skirt_side_bk_l_parent, swing1LimitDegrees=15, swing2LimitDegrees=25, twistLimitDegrees=2}
 {bone1Name=dyn_skirt_side_bk_r_02, bone2Name=dyn_skirt_side_bk_r_01, swing1LimitDegrees=25, swing2LimitDegrees=15, twistLimitDegrees=2}
 {bone1Name=dyn_skirt_side_bk_r_01, bone2Name=dyn_skirt_side_bk_r_parent, swing1LimitDegrees=15, swing2LimitDegrees=25, twistLimitDegrees=2}
 {bone1Name=dyn_skirt_side_frnt_l_02, bone2Name=dyn_skirt_side_frnt_l_01, swing1LimitDegrees=25, swing2LimitDegrees=20, twistLimitDegrees=2}
 {
  bone1Name=dyn_skirt_side_frnt_l_01
  bone2Name=dyn_skirt_side_frnt_l_parent
  swing1LimitDegrees=15
  swing2LimitDegrees=10
  twistLimitDegrees=2
 }
 {bone1Name=dyn_skirt_side_frnt_r_02, bone2Name=dyn_skirt_side_frnt_r_01, swing1LimitDegrees=25, swing2LimitDegrees=20, twistLimitDegrees=2}
 {
  bone1Name=dyn_skirt_side_frnt_r_01
  bone2Name=dyn_skirt_side_frnt_r_parent
  swing1LimitDegrees=15
  swing2LimitDegrees=10
  twistLimitDegrees=2
 }
 {bone1Name=dyn_skirt_side_l_02, bone2Name=dyn_skirt_side_l_01, swing1LimitDegrees=25, swing2LimitDegrees=20, twistLimitDegrees=2}
 {bone1Name=dyn_skirt_side_l_01, bone2Name=dyn_skirt_side_l_parent, swing1LimitDegrees=15, swing2LimitDegrees=10, twistLimitDegrees=2}
 {bone1Name=dyn_skirt_side_r_02, bone2Name=dyn_skirt_side_r_01, swing1LimitDegrees=25, swing2LimitDegrees=20, twistLimitDegrees=2}
 {bone1Name=dyn_skirt_side_r_01, bone2Name=dyn_skirt_side_r_parent, swing1LimitDegrees=15, swing2LimitDegrees=10, twistLimitDegrees=2}
 {bone1Name=dyn_skirt_frnt_mid_parent, bone2Name=dyn_skirt_frnt_mid_base, swing1LimitDegrees=5, swing2LimitDegrees=5, twistLimitDegrees=2}
 {bone1Name=dyn_skirt_side_r_parent, bone2Name=dyn_skirt_side_r_base, swing1LimitDegrees=5, swing2LimitDegrees=5, twistLimitDegrees=2}
 {bone1Name=dyn_skirt_side_l_parent, bone2Name=dyn_skirt_side_l_base, swing1LimitDegrees=5, swing2LimitDegrees=5, twistLimitDegrees=2}
 {
  bone1Name=dyn_skirt_side_frnt_r_parent
  bone2Name=dyn_skirt_side_frnt_r_base
  swing1LimitDegrees=5
  swing2LimitDegrees=5
  twistLimitDegrees=2
 }
 {
  bone1Name=dyn_skirt_side_frnt_l_parent
  bone2Name=dyn_skirt_side_frnt_l_base
  swing1LimitDegrees=5
  swing2LimitDegrees=5
  twistLimitDegrees=2
 }
 {bone1Name=dyn_skirt_side_bk_r_parent, bone2Name=dyn_skirt_side_bk_r_base, swing1LimitDegrees=5, swing2LimitDegrees=5, twistLimitDegrees=2}
 {bone1Name=dyn_skirt_side_bk_l_parent, bone2Name=dyn_skirt_side_bk_l_base, swing1LimitDegrees=5, swing2LimitDegrees=5, twistLimitDegrees=2}
 {bone1Name=dyn_skirt_bk_mid_parent, bone2Name=dyn_skirt_bk_mid_base, swing1LimitDegrees=5, swing2LimitDegrees=5, twistLimitDegrees=2}
]
