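# Animation Blueprint for the female Ballerina player body that copies the attached parent's pose and adds skirt motion and corrections.
# Knee-slide offsets on the rear skirt bones feed a pelvis-space RigidBody simulation, followed by a skydive-weighted Control Rig and an always-on Control Rig.
# Crouch and baseball-slide bone offsets blend in afterward, then a constant rear-skirt pitch.
# EventGraph converts skydive, knee-slide and cape-backbling states to floats, though AnimGraph never reads Cape.
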AnimBlueprint /Game/Characters/Player/Female/Medium/Bodies/F_MED_Ballerina/Meshes/F_MED_Ballerina_AnimBP
target_skeleton=&/F_MED_Ballerina_Skeleton
graphs [
 AnimGraph {nodes [
   n1 = Get KneeSliding() -> (KneeSliding:double)
   n2 = "Transform (Modify) Bone - Bone: dyn_skirt_bk_mid_base"(
    ComponentPose=Local To Component(LocalPose=Copy Pose From Mesh() {bUseAttachedParent=true})
    Alpha=@n1
    Translation="0,0,0"
    Rotation="-30,0,0"
    Scale="1,1,1"
   ) {
    rotationMode=BMM_Additive
    rotationSpace=BCS_BoneSpace
    translationMode=BMM_Additive
    translationSpace=BCS_BoneSpace
    alphaBoolBlend {blendOption=Linear}
    boneToModify {boneName=dyn_skirt_bk_mid_base}
    rotation {pitch=-30}
   }
   n3 = "Transform (Modify) Bone - Bone: dyn_skirt_side_bk_l_base"(ComponentPose=@n2, Alpha=@n1, Translation="0,0,0", Rotation="-30,0,0", Scale="1,1,1") {
    rotationMode=BMM_Additive
    rotationSpace=BCS_BoneSpace
    translationMode=BMM_Additive
    translationSpace=BCS_BoneSpace
    alphaBoolBlend {blendOption=Linear}
    boneToModify {boneName=dyn_skirt_side_bk_l_base}
    rotation {pitch=-30}
   }
   n4 = "Transform (Modify) Bone - Bone: dyn_skirt_side_bk_r_base"(ComponentPose=@n3, Alpha=@n1, Translation="0,0,0", Rotation="-40,0,0", Scale="1,1,1") {
    rotationMode=BMM_Additive
    rotationSpace=BCS_BoneSpace
    translationMode=BMM_Additive
    translationSpace=BCS_BoneSpace
    alphaBoolBlend {blendOption=Linear}
    boneToModify {boneName=dyn_skirt_side_bk_r_base}
    rotation {pitch=-40}
   }
   n5 = RigidBody(ComponentPose=@n4) {
    bClampLinearTranslationLimitToRefPose=true
    bOverrideWorldGravity=true
    overridePhysicsAsset=&/F_MED_Ballerina_Physics
    simulationSpace=BaseBoneSpace
    alphaBoolBlend {blendOption=Linear}
    baseBoneRef {boneName=pelvis}
    componentAppliedLinearAccClamp=<3000 3000 3000>
    componentLinearAccScale=<0.2 0.2 0.25>
    componentLinearVelScale=<0.2 0.2 0.25>
    overrideWorldGravity {z=-980}
    simSpaceSettings {maxAngularAcceleration=100, maxAngularVelocity=9, maxLinearAcceleration=300, maxLinearVelocity=400, velocityScaleZ=0.12, worldAlpha=1}
   }
   n6 = Control Rig(Source=Component To Local(ComponentPose=@n5), Alpha=Get Skydiving() -> (Skydiving:double)) {
    alphaBoolBlend {blendOption=Linear}
    controlRigAssetReference {blueprintRigClass=&/F_MED_Ballerina_Skydive_CtrlRig.F_MED_Ballerina_Skydive_CtrlRig_C}
   }
   n7 = Control Rig(Source=@n6) {
    alphaBoolBlend {blendOption=Linear}
    controlRigAssetReference {blueprintRigClass=&/F_MED_Ballerina_CtrlRig.F_MED_Ballerina_CtrlRig_C}
   }
   n8 = Get bIsCrouching()
   n9 = "Transform (Modify) Bone - Bone: dyn_skirt_bk_mid_parent"!(
    ComponentPose=Local To Component(LocalPose=@n7)
    bAlphaBoolEnabled:bool=@n8
    Translation="-0.5,0,3"
    Rotation="5.567,0.030,-0.028"
    Scale="1,1,1"
   ) -> (Pose) {
    alphaInputType=Bool
    rotationMode=BMM_Additive
    rotationSpace=BCS_BoneSpace
    translationMode=BMM_Additive
    translationSpace=BCS_BoneSpace
    alphaBoolBlend {blendInTime=0.4, blendOption=Linear, blendOutTime=0.4}
    boneToModify {boneName=dyn_skirt_bk_mid_parent}
    rotation=<rot 5.5671 0.03 -0.028>
    translation {x=-0.5, z=3}
   }
   n10 = "Transform (Modify) Bone - Bone: dyn_skirt_side_bk_l_parent"!(ComponentPose=@n9, bAlphaBoolEnabled:bool=@n8, Translation="-0.5,0,3", Rotation="2.752,-0.002,-0.027", Scale="1,1,1") -> (Pose) {
    alphaInputType=Bool
    rotationMode=BMM_Additive
    rotationSpace=BCS_BoneSpace
    translationMode=BMM_Additive
    translationSpace=BCS_BoneSpace
    alphaBoolBlend {blendInTime=0.4, blendOption=Linear, blendOutTime=0.4}
    boneToModify {boneName=dyn_skirt_side_bk_l_parent}
    rotation=<rot 2.752 -0.002 -0.027>
    translation {x=-0.5, z=3}
   }
   n11 = "Transform (Modify) Bone - Bone: dyn_skirt_side_bk_r_parent"!(ComponentPose=@n10, bAlphaBoolEnabled:bool=@n8, Translation="-0.5,0,3", Rotation="5.564,0.009,-0.013", Scale="1,1,1") -> (Pose) {
    alphaInputType=Bool
    rotationMode=BMM_Additive
    rotationSpace=BCS_BoneSpace
    translationMode=BMM_Additive
    translationSpace=BCS_BoneSpace
    alphaBoolBlend {blendInTime=0.4, blendOption=Linear, blendOutTime=0.4}
    boneToModify {boneName=dyn_skirt_side_bk_r_parent}
    rotation=<rot 5.5641 0.009 -0.013>
    translation {x=-0.5, z=3}
   }
   n12 = Get bIsBaseballSliding()
   n13 = "Transform (Modify) Bone - Bone: dyn_skirt_side_frnt_r_base"!(ComponentPose=@n11, bAlphaBoolEnabled:bool=@n12, Translation="18.877,12.779,1.671", Rotation="80.278,-117.784,69.902", Scale="1,1,1") -> (Pose) {
    alphaInputType=Bool
    rotationMode=BMM_Additive
    rotationSpace=BCS_BoneSpace
    translationMode=BMM_Additive
    translationSpace=BCS_BoneSpace
    alphaBoolBlend {blendInTime=0.4, blendOption=Linear, blendOutTime=0.4}
    boneToModify {boneName=dyn_skirt_side_frnt_r_base}
    rotation=<rot 80.278 -117.784 69.902>
    translation=<18.877 12.779 1.671>
   }
   n14 = "Transform (Modify) Bone - Bone: dyn_skirt_side_r_base"!(ComponentPose=@n13, bAlphaBoolEnabled:bool=@n12, Translation="-2.476,-1.915,-4.927", Rotation="-32.189,12.013,47.437", Scale="1,1,1") -> (Pose) {
    alphaInputType=Bool
    rotationMode=BMM_Additive
    rotationSpace=BCS_BoneSpace
    translationMode=BMM_Additive
    translationSpace=BCS_BoneSpace
    alphaBoolBlend {blendInTime=0.4, blendOption=Linear, blendOutTime=0.4}
    boneToModify {boneName=dyn_skirt_side_r_base}
    rotation=<rot -32.1889 12.013 47.437>
    translation=<-2.476 -1.915 -4.927>
   }
   n15 = "Transform (Modify) Bone - Bone: dyn_skirt_frnt_mid_base"!(ComponentPose=@n14, bAlphaBoolEnabled:bool=@n12, Translation="7.150,-5.554,5.333", Rotation="4.395,-34.101,23.438", Scale="1,1,1") -> (Pose) {
    alphaInputType=Bool
    rotationMode=BMM_Additive
    rotationSpace=BCS_BoneSpace
    translationMode=BMM_Additive
    translationSpace=BCS_BoneSpace
    alphaBoolBlend {blendInTime=0.4, blendOption=Linear, blendOutTime=0.4}
    boneToModify {boneName=dyn_skirt_frnt_mid_base}
    rotation=<rot 4.395 -34.101 23.438>
    translation=<7.15 -5.554 5.333>
   }
   n16 = "Transform (Modify) Bone - Bone: dyn_skirt_side_bk_r_base"!(ComponentPose=@n15, bAlphaBoolEnabled:bool=@n12, Translation="-1.661,0.008,-2.079", Rotation="0.164,2.265,-4.879", Scale="1,1,1") -> (Pose) {
    alphaInputType=Bool
    rotationMode=BMM_Additive
    rotationSpace=BCS_BoneSpace
    translationMode=BMM_Additive
    translationSpace=BCS_BoneSpace
    alphaBoolBlend {blendInTime=0.4, blendOption=Linear, blendOutTime=0.4}
    boneToModify {boneName=dyn_skirt_side_bk_r_base}
    rotation=<rot 0.164 2.265 -4.879>
    translation=<-1.661 0.008 -2.079>
   }
   n17 = "Transform (Modify) Bone - Bone: dyn_skirt_side_bk_l_base"!(ComponentPose=@n16, Alpha=1, Rotation="10,0,0") -> (Pose) {
    rotationMode=BMM_Additive
    rotationSpace=BCS_BoneSpace
    translationMode=BMM_Additive
    translationSpace=BCS_BoneSpace
    alphaBoolBlend {blendInTime=0.4, blendOption=Linear, blendOutTime=0.4}
    boneToModify {boneName=dyn_skirt_side_bk_l_base}
    rotation {pitch=10}
   }
   n18 = "Transform (Modify) Bone - Bone: dyn_skirt_bk_mid_base"!(ComponentPose=@n17, Alpha=1, Rotation="10,0,0") -> (Pose) {
    rotationMode=BMM_Additive
    rotationSpace=BCS_BoneSpace
    translationMode=BMM_Additive
    translationSpace=BCS_BoneSpace
    alphaBoolBlend {blendInTime=0.4, blendOption=Linear, blendOutTime=0.4}
    boneToModify {boneName=dyn_skirt_bk_mid_base}
    rotation {pitch=10}
   }
   n19 = "Transform (Modify) Bone - Bone: dyn_skirt_side_bk_r_base"!(ComponentPose=@n18, Alpha:float=1, Rotation="10,0,0") -> (Pose) {
    rotationMode=BMM_Additive
    rotationSpace=BCS_BoneSpace
    translationMode=BMM_Additive
    translationSpace=BCS_BoneSpace
    alphaBoolBlend {blendInTime=0.4, blendOption=Linear, blendOutTime=0.4}
    boneToModify {boneName=dyn_skirt_side_bk_r_base}
    rotation {pitch=10}
   }
   Output Pose(Result=Component To Local(ComponentPose=@n19)) {layerGroup=None}
  ]}
 EventGraph {nodes [
   Event BlueprintUpdateAnimation()
   > n1 = Sequence()
   > Set Skydiving(Skydiving:double=SelectFloat(A=1, B=0, bPickA=Get bIsSkyDiving())) -> (Output_Get:double)
   Set KneeSliding(execute=@n1.then_1, KneeSliding:double=SelectFloat(A=1, B=0, bPickA=Get bIsKneeSliding())) -> (Output_Get:double)
   > Set Cape(Cape:double=SelectFloat(A=1, B=0, bPickA=CheckForBackbling().bIsBackblingACape)) -> (Output_Get:double)
  ]}
]
variables [
 Body_Mesh {type=SkeletalMeshComponent}
 Cape {type=number}
 KneeSliding {type=number}
 Skydiving {type=number}
]
